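; Blueprint waterfall generator that configures a particle-based waterfall effect and its collision.
; The construction script calls its parent, then pushes top position, bottom position and fall width variables into the WaterfallFX particle system as named parameters.
; When UseConstructCollisionOffset is set, it also applies a stored relative transform to the collision component.
; BeginPlay only calls the parent and Tick is empty, so all setup happens at construction.
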(event EventBeginPlay
  (|CallParentFunction))

(event EventTick
  (params (DeltaSeconds :type Float%20%28single-precision%29)))

(fn ConstructionScript
  (bind _waterfallfx (Variables|Default|GetWaterfallFX))
  (|CallParentFunction)
  (Effects|Components|ParticleSystem|SetVectorParameter _waterfallfx "TopPosition" (Variables|Default|GetTopPosition))
  (Effects|Components|ParticleSystem|SetVectorParameter _waterfallfx "BottomPosition" (Variables|Default|GetBottomPosition))
  (Effects|Components|ParticleSystem|SetFloatParameter (Variables|Default|GetWaterfallFX) "FallWidth" (Variables|Default|Getfall%20width))
  (if (Variables|Default|GetUseConstructCollisionOffset)
    (Transformation|SetRelativeTransform (Variables|Default|GetCollision) (Variables|Default|GetCollisionTransform) false false)))
