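# Material for Apollo waterfall cliff meshes that blends tiled cliff textures into the landscape.
# Cliff base color is shaded by world height, then blended with the landscape runtime virtual texture by height above the RVT ground, masked by vertex red.
# Vertex blue darkens base color and remaps roughness, apparently for wetness.
# Without virtual texturing it uses plain cliff; Forward and Mobile paths use a shared mobile cliff function.
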Material /Clyde/Athena/Apollo/Environments/Meshes/Waterfall/M_Apollo_WaterfallCliff_Blend
graph {
 properties {bTangentSpaceNormal=false, bUseMaterialAttributes=true, pixelDepthOffsetMode=PDOM_Legacy, usedShadingModels=Default Lit}
 nodes [
  n1 = Add(A=Multiply(A=TextureCoordinate(), B=Constant() {r=2}), B=Constant2Vector() {g=-0.05})
  n2 = TextureSample(UVs=@n1) {
   samplerSource=SSM_Wrap_WorldGroupSettings
   samplerType=SAMPLERTYPE_Masks
   texture=/Game/Environments/Apollo/Sets/Terrain/Textures/T_Apollo_Cliff_S_03
  }
  n3 = WorldPosition()
  n4 = Add(A=OneMinus(None=ComponentMask(None=@n3.XYZ) {b=true}), B=Constant() {r=10000})
  n5 = Constant() {r=30000}
  n6 = Add(A=Saturate(None=Divide(A=@n4, B=@n5)), B=Constant() {r=0.35})
  n7 = Constant() {r=0.75}
  n8 = TextureSample(UVs=@n1) {samplerSource=SSM_Wrap_WorldGroupSettings, texture=/Game/Environments/Apollo/Sets/Terrain/Textures/T_Apollo_Cliff_D_03}
  n9 = TextureSample(UVs=@n1) {
   samplerSource=SSM_Wrap_WorldGroupSettings
   samplerType=SAMPLERTYPE_Normal
   texture=/Game/Environments/Apollo/Sets/Terrain/Textures/T_Apollo_Cliff_N_03
  }
  n10 = SetMaterialAttributes!(
   MaterialAttributes
   Normal=Transform(None=@n9.RGB)
   Roughness=Multiply(A=Constant() {r=0.8}, B=@n2.B)
   Specular=Multiply(A=Constant() {r=0.3}, B=@n2.R)
   Base Color=Multiply(A=Saturate(None=@n6), B=Multiply(A=@n7, B=@n8.RGB))
  ) {attributeSetTypes [
    #1
    #2
    #3
    #4
   ]}
  n11 = RuntimeVirtualTextureSample!(UVs, World Position) {materialType=BaseColor_Normal_Specular_YCoCg, virtualTexture=/Game/Athena/Apollo/RVT/RVT_Apollo_Landscape}
  n12 = SetMaterialAttributes!(MaterialAttributes, Base Color=@n11.BaseColor, Specular=@n11.BaseColor, Roughness=@n11.BaseColor, Normal=@n11.BaseColor) {attributeSetTypes [
    #4
    #3
    #2
    #1
   ]}
  n13 = WorldPosition() {worldPositionShaderOffset=WPT_ExcludeAllShaderOffsets}
  n14 = Subtract(
   A=Add(A=ComponentMask(None=@n13.XYZ) {b=true}, B=ScalarParameter() {defaultValue=-8000, group=RVT, parameterName=RVT WP Offset})
   B=RuntimeVirtualTextureSample() {materialType=WorldHeight, virtualTexture=/Game/Athena/Apollo/RVT/RVT_Apollo_Landscape_Height}.WorldHeight
  )
  n15 = ScalarParameter() {defaultValue=200, group=RVT, parameterName=Ground fade}
  n16 = VertexColor()
  n17 = LinearInterpolate(B=Saturate(None=Divide(A=@n14, B=@n15)), Alpha=Max(A=@n16.R, B=@n16.R)) {constA=1}
  n18 = GetMaterialAttributes(None=VirtualTextureFeatureSwitch!(No=@n10, Yes=BlendMaterialAttributes(A=@n12, B=@n10, Alpha=@n17))) {attributeGetTypes [#4, #2]}
  n19 = OneMinus(None=VertexColor().B)
  n20 = Constant() {r=0.6}
  n21 = Constant() {r=-1}
  n22 = Constant() {r=0.35}
  n23 = ComponentMask(None=@n18.Roughness) {r=true}
  n24 = SetMaterialAttributes!(
   MaterialAttributes=@n18.MaterialAttributes
   Base Color=Multiply(A=@n18.Base Color, B=Clamp(None=@n19, Min=@n20) {clampMode=CMODE_ClampMin, minDefault=0.65})
   Roughness=LinearInterpolate(A=LinearInterpolate(A=@n21, B=@n22, Alpha=@n23), B=@n23, Alpha=@n19)
  ) {attributeSetTypes [#4, #2]}
  n25 = MaterialFunctionCall!(UseParameters) {materialFunction=/Game/Athena/Apollo/Environments/Landscape/Materials/Biome/CommonMaterials/Mobile/MF_Mobile_Common_Cliff}
  ShadingPathSwitch(Default=@n24, Deferred=@n24, Forward=@n25.LandscapeMaterial, Mobile=@n25.LandscapeMaterial)
 ]
}
